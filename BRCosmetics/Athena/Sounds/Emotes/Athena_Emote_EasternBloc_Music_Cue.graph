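# Sound cue for the Eastern Bloc emote music, using the first-person emote music sound class and concurrency rule.
# After a short initial delay, a mixer starts a half-volume intro one-shot alongside the looping main track, which has its own brief extra delay.
# It has a concurrency override limiting it to a single instance and sends to the music audio-analysis submix.
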SoundCue /BRCosmetics/Athena/Sounds/Emotes/Athena_Emote_EasternBloc_Music_Cue
root=Delay_2
nodes [
 {id=Delay_1, children [WavePlayer_1], properties {delayMax=0.13, delayMin=0.13}}
 {id=Delay_2, children [Mixer_0], properties {delayMax=0.25, delayMin=0.25}}
 {id=Mixer_0, children [Delay_1,WavePlayer_3], properties {inputVolume [1,0.5]}}
 {id=WavePlayer_1, properties {bLooping=true, soundWaveAssetPtr=/Game/Athena/Sounds/Emotes/eastern_bloc_musc_setup_D}}
 {id=WavePlayer_3, properties {soundWaveAssetPtr=/Game/Athena/Sounds/Emotes/eastern_bloc_musc_setup_intro}}
]
properties {
 maxDistance=2097152
 soundClassObject=/Game/Sounds/FortSoundClassesAndModes/Fort_Music_Emote_1P_PSM
 concurrencyOverrides {maxCount=1}
 concurrencySet [/Game/Sounds/Concurrency_Rules/Athena_EmoteMusic_1P_Conc]
 soundSubmixSends [{minSendLevel=0, soundSubmix=&Fortnite_AudioAnalysis_Music}]
}
