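# Sound cue for the rock-paper emote's foley hits in Athena, playing three hits in quick, evenly spaced succession.
# A mixer triggers one shared modulator and random branch immediately and through two fixed delays, drawing from three hit waves.
# It uses emote attenuation, the emote music sound class and concurrency set, and a send to the music audio-analysis submix.
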SoundCue /BRCosmetics/Athena/Sounds/Emotes/Athena_Emote_Rockpaper_FoleyHit_Cue
root=Mixer_1
nodes [
 {id=Delay_0, children [Modulator_0], properties {delayMax=0.3, delayMin=0.3}}
 {id=Delay_1, children [Modulator_0], properties {delayMax=0.6, delayMin=0.6}}
 {id=Mixer_1, children [Modulator_0,Delay_0,Delay_1], properties {inputVolume [1,1,1]}}
 {id=Modulator_0, children [Random_0]}
 {id=Random_0, children [WavePlayer_0,WavePlayer_1,WavePlayer_2], properties {weights [1,1,1]}}
 {id=WavePlayer_0, properties {soundWaveAssetPtr=/Game/Athena/Sounds/Emotes/emote_rockpaper_foley_hit_01}}
 {id=WavePlayer_1, properties {soundWaveAssetPtr=/Game/Athena/Sounds/Emotes/emote_rockpaper_foley_hit_02}}
 {id=WavePlayer_2, properties {soundWaveAssetPtr=/Game/Athena/Sounds/Emotes/emote_rockpaper_foley_hit_03}}
]
properties {
 attenuationSettings=&Fort_Emotes_Athena_Att
 maxDistance=1600
 soundClassObject=&Fort_Music_Emote
 volumeMultiplier=0.6
 concurrencyOverrides {resolutionRule=PreventNew}
 concurrencySet [/Game/Sounds/Concurrency_Rules/Athena_EmoteMusic_Conc]
 soundSubmixSends [{minSendLevel=0, soundSubmix=&Fortnite_AudioAnalysis_Music}]
}
