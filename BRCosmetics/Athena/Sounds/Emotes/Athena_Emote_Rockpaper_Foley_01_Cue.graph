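# Foley sound cue for the rock-paper emote, playing short clothing/movement rustles borrowed from the player run foley set.
# A Random node picks evenly among six run foley waves, and a Modulator raises pitch within a narrow random range on top of a raised cue pitch at low volume.
# It uses emote attenuation, music-emote sound class and concurrency, and sends to the music audio-analysis submix.
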SoundCue /BRCosmetics/Athena/Sounds/Emotes/Athena_Emote_Rockpaper_Foley_01_Cue
root=Modulator_1
nodes [
 {id=Modulator_1, children [Random_1], properties {pitchMax=1.3, pitchMin=1.1}}
 {
  id=Random_1
  children [WavePlayer_9, WavePlayer_10, WavePlayer_11, WavePlayer_12, WavePlayer_13, WavePlayer_14]
  properties {weights [1,1,1,1,1,1]}
 }
 {id=WavePlayer_10, properties {soundWaveAssetPtr=/Game/Sounds/Fort_Foley_Sounds/Moves/player_run_foley_03}}
 {id=WavePlayer_11, properties {soundWaveAssetPtr=/Game/Sounds/Fort_Foley_Sounds/Moves/player_run_foley_02}}
 {id=WavePlayer_12, properties {soundWaveAssetPtr=/Game/Sounds/Fort_Foley_Sounds/Moves/player_run_foley_04}}
 {id=WavePlayer_13, properties {soundWaveAssetPtr=/Game/Sounds/Fort_Foley_Sounds/Moves/player_run_foley_05}}
 {id=WavePlayer_14, properties {soundWaveAssetPtr=/Game/Sounds/Fort_Foley_Sounds/Moves/player_run_foley_06}}
 {id=WavePlayer_9, properties {soundWaveAssetPtr=/Game/Sounds/Fort_Foley_Sounds/Moves/player_run_foley_01}}
]
properties {
 attenuationSettings=&Fort_Emotes_Athena_Att
 maxDistance=1600
 pitchMultiplier=1.2
 soundClassObject=&Fort_Music_Emote
 volumeMultiplier=0.3
 concurrencyOverrides {resolutionRule=PreventNew}
 concurrencySet [/Game/Sounds/Concurrency_Rules/Athena_EmoteMusic_Conc]
 soundSubmixSends [{minSendLevel=0, soundSubmix=&Fortnite_AudioAnalysis_Music}]
}
