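# Third-person sound cue for the Shaolin emote's music and effects, heard by others in Athena through emote attenuation and the 3P emote music class.
# A mixer plays a one-shot intro SFX and music wave alongside a looping music wave delayed by a fixed interval, forming an intro-into-loop handoff.
# The cue keeps playing when silent, uses emote music concurrency, and feeds a music analysis submix.
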SoundCue /BRCosmetics/Athena/Sounds/Emotes/SC_Emote_Shaolin_Cue_3p
root=Mixer_0
nodes [
 {id=Delay_0, children [WavePlayer_1], properties {delayMax=1.3, delayMin=1.3}}
 {id=Mixer_0, children [WavePlayer_0,Delay_0], properties {inputVolume [1,1]}}
 {id=WavePlayer_0, properties {soundWaveAssetPtr=/Game/Athena/Sounds/Emotes/Emote_Shaolin_SFX_MX}}
 {id=WavePlayer_1, properties {bLooping=true, soundWaveAssetPtr=/Game/Athena/Sounds/Emotes/Emote_Shaolin_MX_Loop}}
]
properties {
 attenuationSettings=&Fort_Emotes_Athena_Att
 maxDistance=1600
 soundClassObject=/Game/Sounds/FortSoundClassesAndModes/Fort_Music_Emote_3P
 virtualizationMode=PlayWhenSilent
 volumeMultiplier=1
 concurrencySet [/Game/Sounds/Concurrency_Rules/Athena_EmoteMusic_Conc]
 soundSubmixSends [{minSendLevel=0, soundSubmix=&Fortnite_AudioAnalysis_Music}]
}
